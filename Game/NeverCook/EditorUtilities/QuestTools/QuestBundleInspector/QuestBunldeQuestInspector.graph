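# Editor utility widget that lays out one quest's inspector panel for the quest bundle tools, based on the native QuestBundleQuestInspector class.
# A dark bordered vertical stack holds the quest name, thin dividers, and inherited property detail views for bundle, quest and rewards component.
# The rewards view shows all properties but skips property path generation for ObjectivesExtensionData.
# The event graph holds only unconnected stock events, so the asset itself adds no logic.
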EditorUtilityWidgetBlueprint /Game/NeverCook/EditorUtilities/QuestTools/QuestBundleInspector/QuestBunldeQuestInspector
graphs [EventGraph {nodes [
   Event PreConstruct()
   Event Construct()
   Event Tick()
  ]}]
variables []
widgets {
 info {inheritedWidgetCount=4, parentClass=/Script/FortniteEditor.QuestBundleQuestInspector, rootWidgetClass=&Border, widgetCount=9}
 widgets [
  Border_2 {
   widgetClassPath=&Border
   widget=Border_2
   properties {visibility=SelfHitTestInvisible, brushColor=<rgba 0.0176 0.0176 0.0176 1>, padding=<ltrb 0 2 0 2>}
  }
  VerticalBox_14 {
   widgetClassPath=&VerticalBox
   parent=Border_2
   widget=VerticalBox_14
   slot {class=BorderSlot, properties {padding=<ltrb 0 2 0 2>}}
  }
  HorizontalBox_0 {
   widgetClassPath=&HorizontalBox
   parent=VerticalBox_14
   widget=HorizontalBox_0
   slot {class=VerticalBoxSlot, properties {padding=<ltrb 2 2 2 2>}}
  }
  Text_QuestName {
   widgetClassPath=&TextBlock
   bInherited=true
   parent=HorizontalBox_0
   widget=Text_QuestName
   properties {
    text=Quest Name
    colorAndOpacity {colorUseRule=UseColor_Specified, specifiedColor=<rgba 0.5271 0.5271 0.5271 1>}
    font {
     bForceMonospaced=false
     bMaterialIsStencil=false
     fontMaterial=null
     fontObject=&Roboto
     letterSpacing=0
     monospacedWidth=1
     size=12
     skewAmount=0
     typefaceFontName=Regular
     outlineSettings {
      bApplyOutlineToDropShadows=false
      bMiteredCorners=false
      bSeparateFillAlpha=false
      outlineMaterial=null
      outlineSize=0
      outlineColor=<rgba 0 0 0 1>
     }
    }
   }
   slot {class=HorizontalBoxSlot, properties {horizontalAlignment=HAlign_Left, verticalAlignment=VAlign_Center, size {sizeRule=Fill, value=1}}}
  }
  Image_148 {
   widgetClassPath=&Image
   bIsVariable=true
   parent=VerticalBox_14
   widget=Image_148
   properties {
    brush {
     bIsDynamicallyLoaded=false
     drawAs=Image
     imageType=NoImage
     mirroring=NoMirror
     resourceName=None
     resourceObject=null
     tiling=NoTile
     imageSize=<32 1>
     margin=<ltrb 0 0 0 0>
     outlineSettings {
      bUseBrushTransparency=false
      roundingType=HalfHeightRadius
      width=0
      color {colorUseRule=UseColor_Specified, specifiedColor=<rgba 0 0 0 0>}
      cornerRadii=<0 0 0 0>
     }
     tintColor {colorUseRule=UseColor_Specified, specifiedColor=<rgba 1 1 1 1>}
     uVRegion {bIsValid=false, max=<0 0>, min=<0 0>}
    }
    colorAndOpacity=<rgba 0.0075 0.0075 0.0075 1>
   }
   slot {class=VerticalBoxSlot}
  }
  PropertiesDetailView_Bundle {
   widgetClassPath=/Script/PropertiesDetailView
   bInherited=true
   bIsVariable=true
   parent=VerticalBox_14
   widget=PropertiesDetailView_Bundle
   properties {visibility=SelfHitTestInvisible, propertyColumns [{fillCoefficient=1, propertiesDisplayData []}]}
   slot {class=VerticalBoxSlot, properties {padding=<ltrb 5 0 0 1>}}
  }
  PropertiesDetailView_Quest {
   widgetClassPath=&PropertiesDetailView
   bInherited=true
   bIsVariable=true
   parent=VerticalBox_14
   widget=PropertiesDetailView_Quest
   properties {visibility=SelfHitTestInvisible, propertyColumns [{fillCoefficient=1, propertiesDisplayData []}]}
   slot {class=VerticalBoxSlot, properties {padding=<ltrb 5 0 0 1>}}
  }
  PropertiesDetailView_RewardsComponent {
   widgetClassPath=&PropertiesDetailView
   bInherited=true
   bIsVariable=true
   parent=VerticalBox_14
   widget=PropertiesDetailView_RewardsComponent
   properties {
    bShowAllProperties=true
    visibility=SelfHitTestInvisible
    propertiesToSkipPropertyPathGeneration [ObjectivesExtensionData]
    propertyColumns [{fillCoefficient=1, propertiesDisplayData []}]
   }
   slot {class=VerticalBoxSlot, properties {padding=<ltrb 5 0 0 0>}}
  }
  Image {
   widgetClassPath=&Image
   bIsVariable=true
   parent=VerticalBox_14
   widget=Image
   properties {
    brush {
     bIsDynamicallyLoaded=false
     drawAs=Image
     imageType=NoImage
     mirroring=NoMirror
     resourceName=None
     resourceObject=null
     tiling=NoTile
     imageSize=<32 10>
     margin=<ltrb 0 0 0 0>
     outlineSettings {
      bUseBrushTransparency=false
      roundingType=HalfHeightRadius
      width=0
      color {colorUseRule=UseColor_Specified, specifiedColor=<rgba 0 0 0 0>}
      cornerRadii=<0 0 0 0>
     }
     tintColor {colorUseRule=UseColor_Specified, specifiedColor=<rgba 1 1 1 1>}
     uVRegion {bIsValid=false, max=<0 0>, min=<0 0>}
    }
    colorAndOpacity=<rgba 0.0075 0.0075 0.0075 1>
   }
   slot {class=VerticalBoxSlot}
  }
 ]
}
